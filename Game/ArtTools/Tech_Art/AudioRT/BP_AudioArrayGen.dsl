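; BP_AudioArrayGen records recent music spectrum and amplitude and republishes delayed copies for audio-reactive materials.
; Audio analysis start begins a roughly 60 Hz timer, except on dedicated servers, sampling MPC_Audio_Reactive_Parameters into wrapping history arrays.
; Each update writes current and three older spectrum samples to MPC_FilteredAudioParameterCollection, with matching amplitudes packed into one vector.
; Analysis stop clears the timers and history after a one-second delay.
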(event Custom|ClearArrays
  (bind _historysamples (Variables|Default|GetHistorySamples))
  ;[1 1 1] clear array on begin
  (Utilities|Array|Resize (Variables|Default|GetHistorySamples) 32)
  (Utilities|Array|Resize (Variables|Default|GetAmplitudeMusicHistorySamples) 32)
  (Variables|Default|SetCurrentSamplePosition 0)
  (for (_ _array_index) _historysamples
    (Utilities|Array|SetArrayElem _historysamples _array_index (Math|Color|LinearColorBlack))
    (Utilities|Array|SetArrayElem (Variables|Default|GetAmplitudeMusicHistorySamples) _array_index)))

(event EventTick
  (params (DeltaSeconds :type Float%20%28single-precision%29)))

(event EventBeginPlay
  (bind _returnvalue (WorldSubsystems|GetFortAudioAnalysisSubsystem))
  (Utilities|FlowControl|Sequence
    (:then_0
      (CallFunction|ClearArrays))
    (:then_1
      (AudioAnalysis|BindEventtoOnAudioAnalysisStarted _returnvalue (delegate Custom|AudioAnalysisStarted))
      (AudioAnalysis|BindEventtoOnAudioAnalysisStopped _returnvalue (delegate Custom|AudioAnalysisStopped)))))

(event Custom|AudioAnalysisStarted
  ;[1 1 1] start timer when audio starts
  (if (Utilities|Time|IsValidTimerHandle (Variables|Default|GetCurrentHandle))
    _
    (elif (Networking|IsDedicatedServer)
      _
      (else
        (Development|PrintString "MakeTimerArrayGen")
        (bind _returnvalue (Utilities|Time|SetTimerbyEvent (delegate Custom|UpdateSamples) 0.016 true))
        (Variables|Default|SetCurrentHandle _returnvalue)))))

(event Custom|AudioAnalysisStopped
  ;[1 1 1] when audio stops, pause for 1 sec then clear timer
  (if (not (Utilities|Time|IsValidTimerHandle (Variables|Default|GetPauseHandle)))
    (bind _returnvalue (Utilities|Time|SetTimerbyEvent (delegate Custom|Pause) 1.0))
    (Variables|Default|SetPauseHandle _returnvalue)))

(event Custom|UpdateSamples
  (bind _historysamples (Variables|Default|GetHistorySamples))
  (bind _returnvalue (Utilities|Array|Length _historysamples))
  (bind _returnvalue_2 (- _returnvalue 1))
  (bind _currentsampleposition (Variables|Default|GetCurrentSamplePosition))
  (bind _returnvalue_3 (- _currentsampleposition 10))
  (bind _out (Math|WrappingModulo _returnvalue_3 _returnvalue_2))
  (bind _returnvalue_4 (- _currentsampleposition 20))
  (bind _out_2 (Math|WrappingModulo _returnvalue_4 _returnvalue_2))
  (bind _returnvalue_5 (- _currentsampleposition 30))
  (bind _out_3 (Math|WrappingModulo _returnvalue_5 _returnvalue_2))
  (bind _returnvalue_6 (Rendering|Material|GetVectorParameterValue "/Game/Athena/Events/AudioAnalysis/MPC_Audio_Reactive_Parameters.MPC_Audio_Reactive_Parameters" "Spectrum_Music" :!class /Script/Engine.KismetMaterialLibrary))
  (bind _returnvalue_7 (Rendering|Material|GetScalarParameterValue "/Game/Athena/Events/AudioAnalysis/MPC_Audio_Reactive_Parameters.MPC_Audio_Reactive_Parameters" "Amplitude_Music" :!class /Script/Engine.KismetMaterialLibrary))
  (Utilities|Array|SetArrayElem _historysamples _currentsampleposition _returnvalue_6)
  (Utilities|Array|SetArrayElem (Variables|Default|GetAmplitudeMusicHistorySamples) _currentsampleposition _returnvalue_7)
  (Rendering|Material|SetVectorParameterValue "/Game/ArtTools/Tech_Art/AudioRT/MPC_FilteredAudioParameterCollection.MPC_FilteredAudioParameterCollection" "SpectrumNow" (Utilities|Array|Get(aref) _historysamples _currentsampleposition) :!class /Script/Engine.KismetMaterialLibrary)
  (Rendering|Material|SetVectorParameterValue "/Game/ArtTools/Tech_Art/AudioRT/MPC_FilteredAudioParameterCollection.MPC_FilteredAudioParameterCollection" "SpectrumMinus1" (Utilities|Array|Get(aref) _historysamples _out) :!class /Script/Engine.KismetMaterialLibrary)
  (Rendering|Material|SetVectorParameterValue "/Game/ArtTools/Tech_Art/AudioRT/MPC_FilteredAudioParameterCollection.MPC_FilteredAudioParameterCollection" "SpectrumMinus2" (Utilities|Array|Get(aref) _historysamples _out_2) :!class /Script/Engine.KismetMaterialLibrary)
  (Rendering|Material|SetVectorParameterValue "/Game/ArtTools/Tech_Art/AudioRT/MPC_FilteredAudioParameterCollection.MPC_FilteredAudioParameterCollection" "SpectrumMinus3" (Utilities|Array|Get(aref) _historysamples _out_3) :!class /Script/Engine.KismetMaterialLibrary)
  (Rendering|Material|SetVectorParameterValue "/Game/ArtTools/Tech_Art/AudioRT/MPC_FilteredAudioParameterCollection.MPC_FilteredAudioParameterCollection" "AmplitudeSamplesPacked" (Utilities|Struct|MakeLinearColor (Utilities|Array|Get(aref) (Variables|Default|GetAmplitudeMusicHistorySamples) _currentsampleposition) (Utilities|Array|Get(aref) (Variables|Default|GetAmplitudeMusicHistorySamples) _out) (Utilities|Array|Get(aref) (Variables|Default|GetAmplitudeMusicHistorySamples) _out_2) (Utilities|Array|Get(aref) (Variables|Default|GetAmplitudeMusicHistorySamples) _out_3)) :!class /Script/Engine.KismetMaterialLibrary)
  ;[1 1 1] INCREMENT!
  (Variables|Default|SetCurrentSamplePosition (Math|Integer|%25%28Integer%29 (+ (Variables|Default|GetCurrentSamplePosition) 1) _returnvalue_2)))

(event Custom|Pause
  ;[1 1 1] actually stop timer
  (Development|PrintString "ClearTimer")
  (Utilities|Time|ClearandInvalidateTimerbyHandle (Variables|Default|GetCurrentHandle))
  (Utilities|Time|ClearandInvalidateTimerbyHandle (Variables|Default|GetPauseHandle))
  (CallFunction|ClearArrays))

(fn ConstructionScript
  (bind _returnvalue (Rendering|Material|CreateDynamicMaterialInstance "/Game/ArtTools/Tech_Art/AudioRT/M_SampleAudio_MPC.M_SampleAudio_MPC" "" "Transient" :!class /Script/Engine.KismetMaterialLibrary))
  (Variables|Default|SetM_SampleAudioMID _returnvalue))
