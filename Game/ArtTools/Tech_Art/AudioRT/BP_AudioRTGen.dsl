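; Blueprint that generates audio-driven render targets on clients while Fortnite audio analysis runs.
; A looping timer of roughly 60 Hz draws an audio-sampling dynamic material into RT_Audio_01, then a propagation material into RT_Audio_02, skipping dedicated servers.
; Analysis start and stop events drive it, with stop clearing the timer and both targets after a one-second delay.
; A separate custom event performs the same draws for Sequencer.
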(event Custom|ClearRenderTargets
  (Rendering|ClearRenderTarget2D "/Game/ArtTools/Tech_Art/AudioRT/RT_Audio_01.RT_Audio_01")
  (Rendering|ClearRenderTarget2D "/Game/ArtTools/Tech_Art/AudioRT/RT_Audio_02.RT_Audio_02"))

(event EventBeginPlay
  (bind _returnvalue (WorldSubsystems|GetFortAudioAnalysisSubsystem))
  (Utilities|FlowControl|Sequence
    (:then_0
      (bind _returnvalue_2 (Rendering|Material|CreateDynamicMaterialInstance "/Game/ArtTools/Tech_Art/AudioRT/M_SampleAudio_MPC.M_SampleAudio_MPC" "" "Transient" :!class /Script/Engine.KismetMaterialLibrary))
      (Variables|Default|SetM_SampleAudioMID _returnvalue_2)
      (Rendering|ClearRenderTarget2D "/Game/ArtTools/Tech_Art/AudioRT/RT_Audio_01.RT_Audio_01")
      (Rendering|ClearRenderTarget2D "/Game/ArtTools/Tech_Art/AudioRT/RT_Audio_02.RT_Audio_02"))
    (:then_1
      (AudioAnalysis|BindEventtoOnAudioAnalysisStarted _returnvalue (delegate Custom|AudioAnalysisStarted))
      (AudioAnalysis|BindEventtoOnAudioAnalysisStopped _returnvalue (delegate Custom|AudioAnalysisStopped)))))

(event Collision|EventActorBeginOverlap
  (params (OtherActor :type Actor%20Object%20Reference)))

(event EventTick
  (params (DeltaSeconds :type Float%20%28single-precision%29)))

(event Custom|updateRenderTargetInSequencer
  (Rendering|DrawMaterialtoRenderTarget "/Game/ArtTools/Tech_Art/AudioRT/RT_Audio_01.RT_Audio_01" (Variables|Default|GetM_SampleAudioMID))
  (Rendering|DrawMaterialtoRenderTarget "/Game/ArtTools/Tech_Art/AudioRT/RT_Audio_02.RT_Audio_02" "/Game/ArtTools/Tech_Art/AudioRT/M_AudioPropogate.M_AudioPropogate"))

(event Custom|Update_RT
  (if (not (Networking|IsDedicatedServer))
    (Rendering|DrawMaterialtoRenderTarget "/Game/ArtTools/Tech_Art/AudioRT/RT_Audio_01.RT_Audio_01" (Variables|Default|GetM_SampleAudioMID))
    (Rendering|DrawMaterialtoRenderTarget "/Game/ArtTools/Tech_Art/AudioRT/RT_Audio_02.RT_Audio_02" "/Game/ArtTools/Tech_Art/AudioRT/M_AudioPropogate.M_AudioPropogate")))

(event Custom|AudioAnalysisStarted
  (if (Utilities|Time|IsValidTimerHandle (Variables|Default|GetCurrentHandle))
    _
    (elif (Networking|IsDedicatedServer)
      _
      (else
        (Development|PrintString "MakeTimer")
        (bind _returnvalue (Utilities|Time|SetTimerbyEvent (delegate Custom|Update_RT) 0.016 true))
        (Variables|Default|SetCurrentHandle _returnvalue)))))

(event Custom|AudioAnalysisStopped
  (if (not (Utilities|Time|IsValidTimerHandle (Variables|Default|GetPauseHandle)))
    (bind _returnvalue (Utilities|Time|SetTimerbyEvent (delegate Custom|Pause) 1.0))
    (Variables|Default|SetPauseHandle _returnvalue)))

(event Custom|Pause
  (Development|PrintString "ClearTimer")
  (Utilities|Time|ClearandInvalidateTimerbyHandle (Variables|Default|GetCurrentHandle))
  (Utilities|Time|ClearandInvalidateTimerbyHandle (Variables|Default|GetPauseHandle))
  (CallFunction|ClearRenderTargets))

(fn UpdateRT
  (if (not (Networking|IsDedicatedServer))
    (Rendering|DrawMaterialtoRenderTarget "/Game/ArtTools/Tech_Art/AudioRT/RT_Audio_01.RT_Audio_01" (Variables|Default|GetM_SampleAudioMID))
    (Rendering|DrawMaterialtoRenderTarget "/Game/ArtTools/Tech_Art/AudioRT/RT_Audio_02.RT_Audio_02" "/Game/ArtTools/Tech_Art/AudioRT/M_AudioPropogate.M_AudioPropogate")))

(fn ConstructionScript
  (bind _returnvalue (Rendering|Material|CreateDynamicMaterialInstance "/Game/ArtTools/Tech_Art/AudioRT/M_SampleAudio_MPC.M_SampleAudio_MPC" "" "Transient" :!class /Script/Engine.KismetMaterialLibrary))
  (Variables|Default|SetM_SampleAudioMID _returnvalue))
