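# Unlit material that appears to propagate audio data along the RT_Audio_01 render target.
# It samples that target at UVs offset by 1/32 in U, shifting content toward +U, and clamps to half-texel edges for a 32-texel grid.
# Four named reroutes build row masks from quantized V via SphereMask, but nothing uses them.
# The graph shows no connection from the texture sample to a material output.
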Material /Game/ArtTools/Tech_Art/AudioRT/M_AudioPropogate
graph {
 properties {pixelDepthOffsetMode=PDOM_Legacy, shadingModel=MSM_Unlit, usedShadingModels=Unlit}
 nodes [
  n1 = TextureCoordinate()
  n2 = Floor(None=Multiply(A=ComponentMask(None=@n1) {g=true}, B=Constant() {r=4}))
  NamedRerouteDeclaration(None=SphereMask(A=@n2, B=Constant(), Radius=Constant() {r=1}, Hardness=Constant() {r=1})) {name=Row_01, nodeColor=<rgba 0 0.7188 1 1>}
  NamedRerouteDeclaration(None=SphereMask(A=@n2, B=Constant() {r=1}, Radius=Constant() {r=1}, Hardness=Constant() {r=1})) {name=Row_02, nodeColor=<rgba 0.5234 1 0 1>}
  NamedRerouteDeclaration(None=SphereMask(A=@n2, B=Constant() {r=2}, Radius=Constant() {r=1}, Hardness=Constant() {r=1})) {name=Row_03, nodeColor=<rgba 0 0.6016 1 1>}
  NamedRerouteDeclaration(None=SphereMask(A=@n2, B=Constant() {r=3}, Radius=Constant() {r=1}, Hardness=Constant() {r=1})) {name=Row_04, nodeColor=<rgba 0 0.25 1 1>}
  n3 = TextureCoordinate()
  n4 = AppendVector(A=Divide() {constB=32}, B=Constant())
  n5 = Divide() {constA=0.5, constB=32}
  TextureSample(UVs=Clamp(None=Subtract(A=@n3, B=@n4), Min=@n5, Max=OneMinus(None=@n5))) {samplerType=SAMPLERTYPE_LinearColor, texture=&/RT_Audio_01}
 ]
}
