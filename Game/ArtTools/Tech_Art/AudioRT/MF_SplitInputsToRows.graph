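# Material function that writes four scalar inputs into four horizontal bands of UV space, outputting one value that varies by row.
# The V coordinate of the texture coordinate input is scaled and floored into a row index.
# Hard-edged SphereMasks compare that index to each row number, producing named reroute row masks.
# Chained lerps use masks to select among the inputs, with the first input reapplied in the last lerp.
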MaterialFunction /Game/ArtTools/Tech_Art/AudioRT/MF_SplitInputsToRows
graph {nodes [
  n1 = Constant()
  n2 = FunctionInput(Preview=@n1) {bUsePreviewValueAsDefault=true, inputName=Input_01, inputType=FunctionInput_Scalar}
  n3 = LinearInterpolate(
   A=LinearInterpolate(
    A=@n2
    B=FunctionInput(Preview=@n1) {bUsePreviewValueAsDefault=true, inputName=Input_02, inputType=FunctionInput_Scalar}
    Alpha=NamedRerouteUsage()
   )
   B=FunctionInput(Preview=@n1) {bUsePreviewValueAsDefault=true, inputName=Input_03, inputType=FunctionInput_Scalar}
   Alpha=NamedRerouteUsage()
  )
  n4 = FunctionInput(Preview=@n1) {bUsePreviewValueAsDefault=true, inputName=Input_04, inputType=FunctionInput_Scalar}
  n5 = NamedRerouteUsage()
  FunctionOutput(None=LinearInterpolate(A=LinearInterpolate(A=@n3, B=@n4, Alpha=@n5), B=@n2, Alpha=NamedRerouteUsage()))
  n6 = FunctionInput(Preview=TextureCoordinate()) {bUsePreviewValueAsDefault=true, inputName=TexCoordInput, inputType=FunctionInput_Vector2, sortPriority=1}
  n7 = Floor(None=Multiply(A=ComponentMask(None=@n6) {g=true}, B=Constant() {r=4}))
  NamedRerouteDeclaration(None=SphereMask(A=@n7, B=Constant(), Radius=Constant() {r=1}, Hardness=Constant() {r=1})) {name=Row_01, nodeColor=<rgba 0 0.7188 1 1>}
  NamedRerouteDeclaration(None=SphereMask(A=@n7, B=Constant() {r=1}, Radius=Constant() {r=1}, Hardness=Constant() {r=1})) {name=Row_02, nodeColor=<rgba 0.5234 1 0 1>}
  NamedRerouteDeclaration(None=SphereMask(A=@n7, B=Constant() {r=2}, Radius=Constant() {r=1}, Hardness=Constant() {r=1})) {name=Row_03, nodeColor=<rgba 0 0.6016 1 1>}
  NamedRerouteDeclaration(None=SphereMask(A=@n7, B=Constant() {r=3}, Radius=Constant() {r=1}, Hardness=Constant() {r=1})) {name=Row_04, nodeColor=<rgba 0 0.25 1 1>}
 ]}
