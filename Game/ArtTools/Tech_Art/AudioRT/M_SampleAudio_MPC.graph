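# Unlit material that packs audio-reactive values from MPC_Audio_Reactive_Parameters into texture data, apparently for an audio render target.
# MF_SplitInputsToRows arranges four inputs per channel: one music spectrum component in red, mix amplitudes and 808 stem in green, other stems and voice-over in blue.
# A floored-UV mask writes these values only into the first of 32 columns, keeping RT_Audio_02's sampled content elsewhere.
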Material /Game/ArtTools/Tech_Art/AudioRT/M_SampleAudio_MPC
graph {
 properties {pixelDepthOffsetMode=PDOM_Legacy, shadingModel=MSM_Unlit, usedShadingModels=Unlit}
 nodes [
  CollectionParameter() {collection=/Game/Athena/Events/AudioAnalysis/MPC_Audio_Reactive_Parameters, parameterName=Amplitude_VO_B}
  n1 = MaterialFunctionCall(Float4=CollectionParameter() {collection=&MPC_Audio_Reactive_Parameters, parameterName=Spectrum_Music}) {materialFunction=&BreakOutFloat4Components}
  n2 = TextureCoordinate()
  n3 = CollectionParameter() {collection=&MPC_Audio_Reactive_Parameters, parameterName=Amplitude_Master}
  n4 = CollectionParameter() {collection=&MPC_Audio_Reactive_Parameters, parameterName=Amplitude_AGV}
  n5 = CollectionParameter() {collection=&MPC_Audio_Reactive_Parameters, parameterName=MusicStem_808}
  n6 = CollectionParameter() {collection=&MPC_Audio_Reactive_Parameters, parameterName=Amplitude_Music}
  n7 = CollectionParameter() {collection=&MPC_Audio_Reactive_Parameters, parameterName=MusicStem_Snare}
  n8 = CollectionParameter() {collection=&MPC_Audio_Reactive_Parameters, parameterName=MusicStem_Vocals}
  n9 = CollectionParameter() {collection=&MPC_Audio_Reactive_Parameters, parameterName=Amplitude_VO_A}
  n10 = CollectionParameter() {collection=&MPC_Audio_Reactive_Parameters, parameterName=MusicStem_Kick}
  n11 = Constant() {r=32}
  n12 = ComponentMask(None=Floor(None=Multiply(A=@n2, B=@n11))) {r=true}
  LinearInterpolate(
   A=TextureSample() {samplerType=SAMPLERTYPE_LinearColor, texture=&/RT_Audio_02}.RGB
   B=MaterialFunctionCall(
    X=MaterialFunctionCall!(Input_02=@n1.G, Input_03=@n1.G, Input_04=@n1.G, Input_01=@n1.G, TexCoordInput=@n2) {materialFunction=&/MF_SplitInputsToRows}.Result
    Y=MaterialFunctionCall(Input_02=@n3, Input_03=@n4, Input_04=@n5, Input_01=@n6, TexCoordInput=@n2) {materialFunction=&MF_SplitInputsToRows}.Result
    Z=MaterialFunctionCall(Input_02=@n7, Input_03=@n8, Input_04=@n9, Input_01=@n10, TexCoordInput=@n2) {materialFunction=&MF_SplitInputsToRows}.Result
   ) {materialFunction=&MakeFloat3}.Result
   Alpha=OneMinus(None=Saturate(None=@n12))
  )
  MaterialFunctionCall() {materialFunction=&MakeFloat3}
  MaterialFunctionCall() {materialFunction=&MakeFloat3}
 ]
}
